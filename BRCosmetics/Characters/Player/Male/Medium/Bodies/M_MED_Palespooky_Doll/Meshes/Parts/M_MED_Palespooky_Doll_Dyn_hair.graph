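# Physics asset for the dynamic hair of the M_MED_Palespooky_Doll player body, with a capsule per hair bone.
# Side, top and back strands form one- to three-bone chains rooted at head_offset, with limited swing and very tight twist.
# Head_offset, spine, right clavicle and right upper arm carry large body capsules, while the head and left arm bones have tiny ones.
# Constraints between these body bones leave swing and twist free.
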PhysicsAsset /BRCosmetics/Characters/Player/Male/Medium/Bodies/M_MED_Palespooky_Doll/Meshes/Parts/M_MED_Palespooky_Doll_Dyn_hair
bodies {
 clavicle_l [{shapeType=Capsule, length=0.505, radius=0.505, rotation=<rot 0 -90 0>}]
 clavicle_r [{shapeType=Capsule, length=10.5, radius=8.75, center=<-7.8406 -0.0001 -0.6452>, rotation=<rot 74.9996 0 0>}]
 dyn_hair_01_bk [{shapeType=Capsule, length=1.75, radius=2.25, center=<7.3359 -1.9108 -0.0001>, rotation=<rot 90 0 0>}]
 dyn_hair_02_bk [{shapeType=Capsule, length=2.75, radius=2, center=<7.4369 1.9814 0.6299>, rotation=<rot 74.9827 -179.9998 149.9994>}]
 dyn_hair_03_bk [{shapeType=Capsule, length=2.25, radius=2.5, center=<11 -0.0001 -0.7347>, rotation=<rot 84.9908 180 180>}]
 dyn_hair_04_bk [{shapeType=Capsule, length=4.25, radius=1.25, center=<6.6285 -0.3678 0.6484>, rotation=<rot 74.2237 -167.3432 -158.281>}]
 dyn_hair_05_bk [{shapeType=Capsule, length=4.75, radius=1.5, center=<9.0425 -0.3236 0.5067>, rotation=<rot 79.9611 177.4966 -172.5596>}]
 dyn_hair_a_01_l [{shapeType=Capsule, length=3.0316, radius=1.8392, center=<3.5449 0 0>, rotation=<rot 90 0 0>}]
 dyn_hair_a_01_r [{shapeType=Capsule, length=2.4154, radius=2.2968, center=<4.3213 0.0001 0.1391>, rotation=<rot 79.9997 0 0>}]
 dyn_hair_a_02_l [{shapeType=Capsule, length=3.7084, radius=1.9635, center=<4.9741 0 0>, rotation=<rot 90 0 0>}]
 dyn_hair_a_02_r [{shapeType=Capsule, length=3.4146, radius=1.7575, center=<5.0802 0 0>, rotation=<rot 90 0 0>}]
 dyn_hair_a_03_l [{shapeType=Capsule, length=2.1309, radius=1.1428, center=<2.4632 0 0.6096>, rotation=<rot 69.9969 -180 -180>}]
 dyn_hair_a_03_r [{shapeType=Capsule, length=2.7377, radius=1.1737, center=<2.4916 0 0>, rotation=<rot 90 0 0>}]
 dyn_hair_b_01_l [{shapeType=Capsule, length=2, radius=3.0115, center=<5.017 -0.0002 0.5653>, rotation=<rot 79.9997 0 0>}]
 dyn_hair_b_01_r [{shapeType=Capsule, length=2.0568, radius=2.5578, center=<3.6524 0 0>, rotation=<rot 90 0 0>}]
 dyn_hair_b_02_l [{shapeType=Capsule, length=3.061, radius=1.5417, center=<4.5931 0.0001 0.451>, rotation=<rot 79.9868 -179.9999 -179.9999>}]
 dyn_hair_b_02_r [{shapeType=Capsule, length=2.7151, radius=2.1779, center=<3.4839 -1.1529 -0.0001>, rotation=<rot 90 90 95.002>}]
 dyn_hair_b_03_r [{shapeType=Capsule, length=3.0444, radius=1.3552, center=<2.586 0.3441 0>, rotation=<rot 90 0 -15.0003>}]
 dyn_hair_c_01_l [{shapeType=Capsule, length=3.3212, radius=1.0223, center=<3.9918 0.0001 -0.1141>, rotation=<rot 84.9776 -180 180>}]
 dyn_hair_c_01_r [{shapeType=Capsule, length=3.1709, radius=2.266, center=<4.8255 0 -0.5262>, rotation=<rot 84.9999 0 0>}]
 dyn_hair_c_02_r [{shapeType=Capsule, length=4.3309, radius=1.581, center=<5.7915 0.4748 0.6164>, rotation=<rot 74.9903 -179.9999 169.9993>}]
 dyn_topHair_a_01_l [{shapeType=Capsule, length=0.7451, radius=1.9978, center=<5.9654 -0.4032 0.4677>, rotation=<rot 84.9971 174.9998 -180>}]
 dyn_topHair_a_01_r [{shapeType=Capsule, length=4.6939, radius=2.4, center=<6.7238 0 -0.5289>, rotation=<rot 79.9999 0 0>}]
 dyn_topHair_a_02_l [{shapeType=Capsule, length=2.8627, radius=1.4063, center=<4.1916 0.3991 -0.2865>, rotation=<rot 79.9996 0 0>}]
 dyn_topHair_b_01_l [{shapeType=Capsule, length=1.5389, radius=1.9249, center=<6.4057 1.0841 0.1708>, rotation=<rot 84.9969 10.001 0>}]
 dyn_topHair_b_01_r [{shapeType=Capsule, length=2.3601, radius=1.7213, center=<4.0919 -0.4713 -0.0338>, rotation=<rot 79.9997 0 10.0006>}]
 dyn_topHair_b_02_l [{shapeType=Capsule, length=2.8905, radius=1.607, center=<4.6598 0.0001 -0.4118>, rotation=<rot 90 -0.0198 -0.0199>}]
 dyn_topHair_b_02_r [{shapeType=Capsule, length=3.1351, radius=1.1056, center=<3.0545 0 0>, rotation=<rot 90 0 0>}]
 dyn_topHair_c_01_l [{shapeType=Capsule, length=1.2299, radius=2.2875, center=<4.4876 0 -0.9529>, rotation=<rot 79.9999 0 0>}]
 dyn_topHair_c_01_r [{shapeType=Capsule, length=1.4092, radius=1.9837, center=<4.4434 0 0.069>, rotation=<rot 79.9999 0 0>}]
 dyn_topHair_c_02_l [{shapeType=Capsule, length=2.4955, radius=1.0487, center=<4.8951 0 -0.1958>, rotation=<rot 90 0 0>}]
 dyn_topHair_c_02_r [{shapeType=Capsule, length=0.7836, radius=2.9038, center=<3.5364 0 0>, rotation=<rot 90 0 0>}]
 dyn_topHair_d_01_l [{shapeType=Capsule, length=3.6985, radius=2.0956, center=<4.5524 0 0.0227>, rotation=<rot 84.9911 180 180>}]
 dyn_topHair_e_01_l [{shapeType=Capsule, length=5.2593, radius=1.7721, center=<5.5975 -0.0002 -0.267>, rotation=<rot 79.9982 0 5>}]
 head [{shapeType=Capsule, length=1, radius=1, rotation=<rot 90 0 0>}]
 head_offset [
  {shapeType=Capsule, length=2.25, radius=13.5, center=<9.1635 0.6593 0.4411>, rotation=<rot 90 0 0>}
  {shapeType=Capsule, length=4.5, radius=11, center=<6.4125 5.1218 -0.1414>, rotation=<rot 90 0 0>}
 ]
 spine_05 [{shapeType=Capsule, length=12.75, radius=9.75, center=<5.9334 -1.5248 0>}]
 upperarm_l [{shapeType=Capsule, length=0.505, radius=0.505, rotation=<rot 0 -90 0>}]
 upperarm_r [{shapeType=Capsule, length=24.75, radius=5.25, center=<-11.7555 -0.0002 0>, rotation=<rot 90 0 0>}]
}
constraints [
 {bone1Name=dyn_topHair_a_01_l, bone2Name=head_offset, swing1LimitDegrees=15, swing2LimitDegrees=30, twistLimitDegrees=2}
 {bone1Name=dyn_topHair_a_02_l, bone2Name=dyn_topHair_a_01_l, swing1LimitDegrees=15, swing2LimitDegrees=30, twistLimitDegrees=2}
 {bone1Name=dyn_topHair_b_01_l, bone2Name=head_offset, swing1LimitDegrees=15, swing2LimitDegrees=30, twistLimitDegrees=2}
 {bone1Name=dyn_topHair_b_02_l, bone2Name=dyn_topHair_b_01_l, swing1LimitDegrees=15, swing2LimitDegrees=30, twistLimitDegrees=2}
 {bone1Name=dyn_topHair_c_01_l, bone2Name=head_offset, swing1LimitDegrees=10, swing2LimitDegrees=20, twistLimitDegrees=2}
 {bone1Name=dyn_topHair_c_02_l, bone2Name=dyn_topHair_c_01_l, swing1LimitDegrees=15, swing2LimitDegrees=30, twistLimitDegrees=2}
 {bone1Name=dyn_topHair_d_01_l, bone2Name=head_offset, swing1LimitDegrees=15, swing2LimitDegrees=30, twistLimitDegrees=2}
 {bone1Name=dyn_topHair_e_01_l, bone2Name=head_offset, swing1LimitDegrees=15, swing2LimitDegrees=30, twistLimitDegrees=2}
 {bone1Name=dyn_topHair_a_01_r, bone2Name=head_offset, swing1LimitDegrees=15, swing2LimitDegrees=30, twistLimitDegrees=2}
 {bone1Name=dyn_topHair_b_01_r, bone2Name=head_offset, swing1LimitDegrees=15, swing2LimitDegrees=30, twistLimitDegrees=2}
 {bone1Name=dyn_topHair_b_02_r, bone2Name=dyn_topHair_b_01_r, swing1LimitDegrees=15, swing2LimitDegrees=30, twistLimitDegrees=2}
 {bone1Name=dyn_topHair_c_01_r, bone2Name=head_offset, swing1LimitDegrees=15, swing2LimitDegrees=30, twistLimitDegrees=2}
 {bone1Name=dyn_topHair_c_02_r, bone2Name=dyn_topHair_c_01_r, swing1LimitDegrees=15, swing2LimitDegrees=30, twistLimitDegrees=2}
 {bone1Name=dyn_hair_a_01_r, bone2Name=head_offset, swing1LimitDegrees=10, swing2LimitDegrees=20, twistLimitDegrees=2}
 {bone1Name=dyn_hair_a_02_r, bone2Name=dyn_hair_a_01_r, swing1LimitDegrees=10, swing2LimitDegrees=20, twistLimitDegrees=2}
 {bone1Name=dyn_hair_a_03_r, bone2Name=dyn_hair_a_02_r, swing1LimitDegrees=10, swing2LimitDegrees=20, twistLimitDegrees=2}
 {bone1Name=dyn_hair_b_01_r, bone2Name=head_offset, swing1LimitDegrees=10, swing2LimitDegrees=20, twistLimitDegrees=2}
 {bone1Name=dyn_hair_b_02_r, bone2Name=dyn_hair_b_01_r, swing1LimitDegrees=10, swing2LimitDegrees=20, twistLimitDegrees=2}
 {bone1Name=dyn_hair_b_03_r, bone2Name=dyn_hair_b_02_r, swing1LimitDegrees=10, swing2LimitDegrees=20, twistLimitDegrees=2}
 {bone1Name=dyn_hair_c_01_r, bone2Name=head_offset, swing1LimitDegrees=10, swing2LimitDegrees=20, twistLimitDegrees=2}
 {bone1Name=dyn_hair_c_02_r, bone2Name=dyn_hair_c_01_r, swing1LimitDegrees=10, swing2LimitDegrees=20, twistLimitDegrees=2}
 {bone1Name=dyn_hair_a_01_l, bone2Name=head_offset, swing1LimitDegrees=10, swing2LimitDegrees=20, twistLimitDegrees=2}
 {bone1Name=dyn_hair_a_02_l, bone2Name=dyn_hair_a_01_l, swing1LimitDegrees=10, swing2LimitDegrees=20, twistLimitDegrees=2}
 {bone1Name=dyn_hair_a_03_l, bone2Name=dyn_hair_a_02_l, swing1LimitDegrees=10, swing2LimitDegrees=20, twistLimitDegrees=2}
 {bone1Name=dyn_hair_b_01_l, bone2Name=head_offset, swing1LimitDegrees=10, swing2LimitDegrees=20, twistLimitDegrees=2}
 {bone1Name=dyn_hair_b_02_l, bone2Name=dyn_hair_b_01_l, swing1LimitDegrees=10, swing2LimitDegrees=20, twistLimitDegrees=2}
 {bone1Name=dyn_hair_c_01_l, bone2Name=head_offset, swing1LimitDegrees=10, swing2LimitDegrees=20, twistLimitDegrees=2}
 {bone1Name=dyn_hair_01_bk, bone2Name=head_offset, swing1LimitDegrees=15, swing2LimitDegrees=20, twistLimitDegrees=2}
 {bone1Name=dyn_hair_02_bk, bone2Name=head_offset, swing1LimitDegrees=15, swing2LimitDegrees=20, twistLimitDegrees=2}
 {bone1Name=dyn_hair_03_bk, bone2Name=head_offset, swing1LimitDegrees=15, swing2LimitDegrees=20, twistLimitDegrees=2}
 {bone1Name=dyn_hair_04_bk, bone2Name=head_offset, swing1LimitDegrees=15, swing2LimitDegrees=20, twistLimitDegrees=2}
 {bone1Name=dyn_hair_05_bk, bone2Name=head_offset, swing1LimitDegrees=15, swing2LimitDegrees=20, twistLimitDegrees=2}
 {
  bone1Name=head
  bone2Name=spine_05
  swing1LimitDegrees=45
  swing1Motion=Free
  swing2LimitDegrees=45
  swing2Motion=Free
  twistLimitDegrees=45
  twistMotion=Free
 }
 {
  bone1Name=head_offset
  bone2Name=spine_05
  swing1LimitDegrees=45
  swing1Motion=Free
  swing2LimitDegrees=45
  swing2Motion=Free
  twistLimitDegrees=45
  twistMotion=Free
 }
 {
  bone1Name=upperarm_l
  bone2Name=clavicle_l
  swing1LimitDegrees=45
  swing1Motion=Free
  swing2LimitDegrees=45
  swing2Motion=Free
  twistLimitDegrees=45
  twistMotion=Free
 }
 {
  bone1Name=upperarm_r
  bone2Name=clavicle_r
  swing1LimitDegrees=45
  swing1Motion=Free
  swing2LimitDegrees=45
  swing2Motion=Free
  twistLimitDegrees=45
  twistMotion=Free
 }
 {
  bone1Name=clavicle_l
  bone2Name=spine_05
  swing1LimitDegrees=45
  swing1Motion=Free
  swing2LimitDegrees=45
  swing2Motion=Free
  twistLimitDegrees=45
  twistMotion=Free
 }
 {
  bone1Name=clavicle_r
  bone2Name=spine_05
  swing1LimitDegrees=45
  swing1Motion=Free
  swing2LimitDegrees=45
  swing2Motion=Free
  twistLimitDegrees=45
  twistMotion=Free
 }
]
