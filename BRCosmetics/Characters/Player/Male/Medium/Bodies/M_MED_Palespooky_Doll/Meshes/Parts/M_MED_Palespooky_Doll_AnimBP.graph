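# Animation Blueprint for the Palespooky Doll medium male player cosmetic part, copying body pose and adding simulated hair motion.
# It layers the pose from Body_Mesh, cached at initialization from the owner's skeletal mesh, onto the attached parent's pose from both upper arms down.
# A constraint then drives head_offset half from head and half from neck_02.
# Finally, a RigidBody node simulates the hair physics asset in spine_05 space with damped component motion.
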AnimBlueprint /BRCosmetics/Characters/Player/Male/Medium/Bodies/M_MED_Palespooky_Doll/Meshes/Parts/M_MED_Palespooky_Doll_AnimBP
target_skeleton=&/M_MED_Palespooky_Doll_Skeleton
graphs [
 AnimGraph {nodes [
   n1 = Copy Pose From Mesh(SourceMeshComponent=Get Body_Mesh())
   n2 = Copy Pose From Mesh() {bUseAttachedParent=true}
   n3 = "Constraint - head_offset (PP)"!(
    ComponentPose=Local To Component(LocalPose=Layered blend per bone(BasePose=@n2, BlendPoses_0=@n1) {layerSetup [{branchFilters [{boneName=upperarm_l}, {boneName=upperarm_r}]}]})
    Alpha=1
    ConstraintWeights_0=0.5
    ConstraintWeights_1:float=0.5
   ) -> (Pose) {
    alphaBoolBlend {blendOption=Linear}
    boneToModify {boneName=head_offset}
    constraintSetup [{targetBone {boneName=head}}, {targetBone {boneName=neck_02}}]
    constraintWeights [1,1]
   }
   n4 = RigidBody(ComponentPose=@n3) {
    bClampLinearTranslationLimitToRefPose=true
    bOverrideWorldGravity=true
    overridePhysicsAsset=&/M_MED_Palespooky_Doll_Dyn_hair
    simulationSpace=BaseBoneSpace
    alphaBoolBlend {blendOption=Linear}
    baseBoneRef {boneName=spine_05}
    componentAppliedLinearAccClamp=<3000 3000 3000>
    componentLinearAccScale=<0.3 0.35 0.35>
    componentLinearVelScale=<0.3 0.3 0.3>
    overrideWorldGravity {z=-980}
   }
   Output Pose(Result=Component To Local(ComponentPose=@n4)) {layerGroup=None}
  ]}
 EventGraph {nodes [
   Event BlueprintInitializeAnimation()
   > Set Body_Mesh(Body_Mesh=GetOwnerSkeletalMeshComponent())
   Event BlueprintUpdateAnimation()
  ]}
]
variables [Body_Mesh {type=SkeletalMeshComponent}]
